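# Sound cue for the looping electrical spark layer of the BariumFemale glider.
# It randomly picks one of three equally weighted looping spark bursts, varies its volume, delays it briefly, then fades it in through an envelope.
# It uses glider thrust attenuation and the glider landing sound class, and borrows the SiennaMale glider's electric concurrency asset.
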SoundCue /BRCosmetics/Sounds/Gliders/BariumFemale/Glider_BariumFemale_Spark_Cue
root=Enveloper_0
nodes [
 {id=Delay_1, children [Modulator_0], properties {delayMax=0.25, delayMin=0.25}}
 {
  id=Enveloper_0
  children [Delay_1]
  properties {pitchCurve {editorCurveData {keys [{time=0}]}}, volumeCurve {editorCurveData {keys [<at 0 0.01>,<at 0.5 1>]}}}
 }
 {id=Modulator_0, children [Random_0], properties {volumeMax=0.6, volumeMin=0.25}}
 {id=Random_0, children [WavePlayer_10, WavePlayer_11, WavePlayer_12], properties {weights [1,1,1]}}
 {id=WavePlayer_10, properties {bLooping=true, soundWaveAssetPtr=&/Glider_Barium_Spark_Burst_03}}
 {id=WavePlayer_11, properties {bLooping=true, soundWaveAssetPtr=&/Glider_Barium_Spark_Burst_01}}
 {id=WavePlayer_12, properties {bLooping=true, soundWaveAssetPtr=&/Glider_Barium_Spark_Burst_02}}
]
properties {
 attenuationSettings=&Fort_Player_Glider_Thrust_Athena_Att
 maxDistance=12500
 priority=0.5
 soundClassObject=&Fort_Gliders_Land
 volumeMultiplier=0.7
 concurrencyOverrides {maxCount=1}
 concurrencySet [/BRCosmetics/Sounds/Gliders/SiennaMale/Glider_SiennaMale_Elec_Concurrency]
}
