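# Niagara spark-burst effect for the red QuietPeanuts emote, in hot orange-yellow.
# HandSparks_Light spawns a brief sprite-plus-light flash that pops in size and computes a waveform-flickered alternate color.
# Speed_Line_Helmet001 bursts a few cone-launched streaks, slowed by drag and gravity, stretched by speed, and fading out.
# Barium-named paint color and audio player variables are set, but no listed module uses them.
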NiagaraSystem /BRCosmetics/Effects/Fort_Effects/Emotes/QuietPeanuts/Red/NS_QuietPeanuts_red_Sparks
dependencies {
 usedDataInterfaces [
  &NiagaraDataInterfaceCurve
  &NiagaraDataInterfaceVector2DCurve
  &NiagaraDataInterfaceColorCurve
  &NiagaraDataInterfaceAudioPlayer
 ]
 usedDynamicInputs [&HueShiftLinearColor]
 usedModules [
  &self:SystemSpawnScript.SystemScriptSource.SystemScriptGraph.NiagaraNodeAssignment_0.SetVariables_#1
  &SystemState
 ]
}
emitters {
 HandSparks_Light {
  emitterSpawnScript [
   /SetVariables(Emitter.EmitRefract=1.5) {isSetParameters=true, module=SetVariables_#2}
   /SetVariables(Emitter.EmitColorLightning:LinearColor=$User.BariumPaintColor) {isSetParameters=true, module=SetVariables_#3}
   /SetVariables(Emitter.AudioPlayer=DataInterface() {propertyValues {
      Attenuation=null
      Concurrency=null
      MaxPlaysPerTick=10
      SoundToPlay=/BRCosmetics/Sounds/Gliders/BariumFemale/Glider_BariumFemale_Spark_Cue
      bAllowLoopingOneShotSounds=false
      bLimitPlaysPerTick=true
      bOnlyActiveDuringGameplay=false
      bStopWhenComponentIsDestroyed=true
      ConfigurationUserParameter {parameter {name=None, typeDefHandle {registeredTypeIndex=1}, varData []}}
      ParameterNames []
     }}) {isSetParameters=true}
  ]
  emitterUpdateScript [
   &EmitterState()
   &SpawnBurst_Instantaneous(Use Spawn Probability=false, UseLoopCountLimit=false)
  ]
  particleSpawnScript [
   &InitializeParticle(
    Lifetime Max=0.3
    Lifetime Min=0.2
    Lifetime Mode=::NewEnumerator1
    Sprite Rotation Mode=::NewEnumerator2
    Sprite Size Max=<30 15>
    Sprite Size Min=<25 10>
    Sprite Size Mode=::NewEnumerator2
    UsePositionOffset=false
    Write Material Random=true
   )
   /SetVariables(Particles.Position2:NiagaraPosition=&Multiply_Vector(A=<0.05 0.05 0.05>, B=$Particles.Velocity)) {isSetParameters=true}
  ]
  particleUpdateScript [
   &ParticleState()
   &ScaleSpriteSize(Uniform Curve Sprite Scale=DataInterface() {propertyValues {Curve {keys [
        {arriveTangent=9.0725, interpMode=RCIM_Cubic, leaveTangent=9.0725, tangentMode=RCTM_User, time=0, value=0}
        {arriveTangent=-2.5249, interpMode=RCIM_Cubic, leaveTangent=-2.5249, tangentMode=RCTM_User, time=0.1408, value=0.9847}
        {arriveTangent=-0.0405, interpMode=RCIM_Cubic, leaveTangent=-0.0405, tangentMode=RCTM_User, time=1, value=0}
       ]}}})
   &SolveForcesAndVelocity(Clamp Velocity=false, Limit Acceleration=false)
   &CameraOffset(Camera Offset Amount=-4)
   &DynamicMaterialParameters(
    Index 0 Param 1=&RandomRangeFloat(Minimum=1, Maximum=1)
    Index 0 Param 2=1
    Module.Param0WriteEnabled=true
    Module.Param1WriteEnabled=true
    Module.Param2WriteEnabled=true
    Module.Param3WriteEnabled=true
   )
   /SetVariables(Particles.Color2=&MakeLinearColorFromVectorAndFloat(
     "Vector (RGB)"=&Multiply_Vector(
      A=&MakeVectorFromLinearColorRGB(Color=<rgba 1 0.8863 0.0937 1>)
      B=&VectorFromFloat(Value=2)
     )
     "Float (Alpha)"=&Waveform(
      Waveform Count=::NewEnumerator2
      Global Amplitude Scale=0.8
      "Waveform [1]"=::NewEnumerator5
      "[1] Frequency"=10
      "[1] Amplitude Min/Max"=<0 1>
      "[1] Bias"=0.2
      "[1] Pulse Duration"=0.2
      "[2] Average"=false
      "[2] Input Position"=0
      "[2] Phase"=$Emitter.Age
      "[2] Phase Scale"=1
      "[2] Compound Frequency"=2
      "[2] Frequency"=14
      "[2] Amplitude Min/Max"=<0 1>
      "[2] Scale"=1
      "[2] Bias"=0.2
      "[2] Exponent"=1
      "[2] Pulse Duration"=0.1
      "[2] Clip Triangle Min/Max"=<0 0>
      "[2] Random Seed"=0
      "[3] Average"=false
      "Waveform [3]"=::NewEnumerator5
      "[3] Input Position"=0
      "[3] Phase"=$Emitter.Age
      "[3] Phase Scale"=1
      "[3] Frequency"=15
      "[3] Compound Frequency"=2
      "[3] Amplitude Min/Max"=<0 1>
      "[3] Scale"=1
      "[3] Bias"=0.5
      "[3] Exponent"=1
      "[3] Pulse Duration"=0.25
      "[3] Clip Triangle Min/Max"=<0 0>
      "[3] Random Seed"=0
     )
    )) {isSetParameters=true}
   &ScaleColor(
    Linear Color Curve=DataInterface() {propertyValues {
      AlphaCurve {keys [<at 0 1>,<at 1 0>]}
      BlueCurve {keys [<at 0 1.1934>,<at 1 0.0312>]}
      GreenCurve {keys [<at 0 13.4664>,<at 1 0.3037>]}
      RedCurve {keys [<at 0 20>,<at 1 1>]}
     }}
    Scale Mode=::NewEnumerator2
   ) {enabled=false}
   &Color(Color=<rgba 1 0.5667 0 1>)
  ]
  rendererClasses [&NiagaraSpriteRendererProperties, &NiagaraLightRendererProperties]
  renderers [
   {rendererClass=&NiagaraSpriteRendererProperties}
   {rendererClass=&NiagaraLightRendererProperties, rendererIndex=1}
  ]
 }
 Speed_Line_Helmet001 {
  emitterSpawnScript [
   /SetVariables(Emitter.EmitRefract=1.5) {isSetParameters=true, module=SetVariables_#2}
   /SetVariables(Emitter.EmitColorLightning:LinearColor=$User.BariumPaintColor) {isSetParameters=true, module=SetVariables_#3}
  ]
  emitterUpdateScript [
   &EmitterState()
   &SpawnBurst_Instantaneous(
    Spawn Count=&RandomRangeInt(Minimum=5, Maximum=8)
    Use Spawn Probability=false
    UseLoopCountLimit=false
   )
  ]
  particleSpawnScript [
   /SetVariables(Particles.SparkIntensity:NiagaraFloat=&FloatFromCurve(
     FloatCurve=DataInterface() {propertyValues {Curve {keys [
         <at 0 1 RCIM_Cubic>
         {arriveTangent=-2.5317, interpMode=RCIM_Cubic, leaveTangent=-2.5317, tangentMode=RCTM_User, time=0.6, value=0.5}
         <at 1 0 RCIM_Cubic>
        ]}}}
     CurveIndex=&RandomRangeFloat(Minimum=0, Maximum=1)
    )) {isSetParameters=true}
   &InitializeParticle(
    Color Minimum=<rgba 0.8 0.16 0 1>
    Color Mode=::NewEnumerator2
    Lifetime Max=0.5
    Lifetime Min=0.02
    Lifetime Mode=::NewEnumerator1
    Sprite Rotation Mode=::NewEnumerator2
    Sprite Size=&MakeVector2D(
     X=&RandomRangeFloat(Minimum=3, Maximum=5)
     Y=&RandomRangeFloat(Minimum=10, Maximum=10)
    )
    Sprite Size Mode=::NewEnumerator4
    UsePositionOffset=false
    Write Material Random=true
   )
   &ShapeLocation(
    Cylinder Height=0
    Cylinder Radius=5
    MaskSpawnGroup=false
    Offset Mode=::NewEnumerator0
    Override Seed=false
    Rotation Axis=<0 0 1>
    Rotation Coordinate Space=::World
    Rotation Mode=::NewEnumerator1
    Shape Primitive=::NewEnumerator1
    Surface Only=false
    UseOwnerScale=true
   )
   &AddVelocity(
    Cone Angle=60
    Module.UseConeFalloff=false
    Module.UseVelDistribution=false
    Override Seed=false
    Velocity Mode=::NewEnumerator2
    Velocity Speed=500
   )
   /SetVariables(Particles.SpriteSize2=&RandomRangeVector2D(Minimum=<5 5>, Maximum=<15 15>)) {isSetParameters=true}
   /SetVariables(Particles.Position=&Add_Vector(
     A=Linked() {linkedVariable {name=Particles.Position, type=NiagaraPosition}}
     B=&TransformVector(Vector=<0 0 -1>, Source Space=::World, Destination Space=::Local)
    )) {isSetParameters=true}
   /SetVariables(Particles.Position2:NiagaraPosition=&Multiply_Vector(A=<0.02 0.02 0.01>, B=$Particles.Velocity)) {isSetParameters=true}
  ]
  particleUpdateScript [
   &ParticleState()
   &GravityForce()
   &Drag(Drag=8, Use Linear Drag=true, Use Rotational Drag=false)
   &SolveForcesAndVelocity(Clamp Velocity=false, Limit Acceleration=false)
   &CameraOffset(Camera Offset Amount=-5)
   &DynamicMaterialParameters(
    Index 0 Param 1=&RandomRangeFloat(Minimum=0.2, Maximum=5)
    Index 0 Param 2=1
    Index 0 Param 3=1
    Module.Param0WriteEnabled=true
    Module.Param1WriteEnabled=true
    Module.Param2WriteEnabled=true
    Module.Param3WriteEnabled=true
   )
   &ScaleSpriteSizeBySpeed(
    Max Scale Factor=<1 5>
    Min Scale Factor=<0.1 0.01>
    Sample Scale Factor By Curve=true
    Scale Factor Curve=DataInterface() {propertyValues {Curve {keys [
        {interpMode=RCIM_Cubic, tangentMode=RCTM_User, time=0, value=0}
        {arriveTangent=2.3236, interpMode=RCIM_Cubic, leaveTangent=2.3236, tangentMode=RCTM_User, time=1, value=1}
       ]}}}
    Velocity Threshold=500
   )
   &ScaleSpriteSize(
    Non-Uniform Curve Sprite Scale=DataInterface() {propertyValues {
      XCurve {keys [<at 0.3 0.9>, {interpMode=RCIM_Cubic, tangentMode=RCTM_User, time=1, value=0.7}]}
      YCurve {keys [
        {tangentMode=RCTM_User, time=0, value=0}
        <at 0.2 1>
        {arriveTangent=-0.0405, interpMode=RCIM_Cubic, leaveTangent=-0.0405, tangentMode=RCTM_User, time=0.7009, value=1}
        {arriveTangent=-8.0089, interpMode=RCIM_Cubic, leaveTangent=-8.0089, tangentMode=RCTM_User, time=1, value=0}
       ]}
     }}
    Scale Sprite Size Mode=::NewEnumerator3
   )
   &ScaleColor(
    Linear Color Curve=DataInterface() {propertyValues {
      AlphaCurve {keys [<at 0 1>,<at 1 0>]}
      BlueCurve {keys [<at 0 0.2983>,<at 1 0.0312>]}
      GreenCurve {keys [<at 0 3.3666>,<at 1 0.3037>]}
      RedCurve {keys [<at 0 5>,<at 1 1>]}
     }}
    Scale Mode=::NewEnumerator2
   )
  ]
  rendererClasses [&NiagaraSpriteRendererProperties]
  renderers [{rendererClass=&NiagaraSpriteRendererProperties}]
 }
}
summary {
 systemName=NS_QuietPeanuts_red_Sparks
 emitters [
  {emitterName=Speed_Line_Helmet001, rendererClasses [&NiagaraSpriteRendererProperties]}
  {
   emitterName=HandSparks_Light
   rendererClasses [&NiagaraSpriteRendererProperties, &NiagaraLightRendererProperties]
  }
 ]
 userVariables [
  User.BariumPaintColor {type=LinearColor, description="", defaultValue {struct=LinearColor, value=<rgba 1 1 1 1>}}
  User.SkeletalMesh {
   type=NiagaraDataInterfaceSkeletalMesh
   defaultValue {
    struct=NiagaraExt_VariableValue_DataInterface
    value {dataInterface=&self:NiagaraDataInterfaceSkeletalMesh_0}
   }
  }
  User.SpawnHelmetFX {type=NiagaraFloat, defaultValue=0, description=""}
 ]
}
